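# Control Rig for the Selene Cobra Stride face accessory part that applies pose-driven corrective offsets to dynamic head bones, apparently hair.
# Spherical pose readers on the head weight bone adjustments, followed by a spherical collision pushing dyn_head_03_r away from spine_05.
# A second sequence branch multiplies right upper arm pose readers by the head look-up alpha for two-handed and rifle aim-up correctives, smoothing the rifle weight with AlphaInterp.
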ControlRigBlueprint /BRCosmetics/Characters/Player/Female/Medium/Bodies/F_MED_SeleneCobra_Stride/Meshes/Parts/F_MED_SeleneCobra_Stride_FaceAcc_CtrlRig
graphs [
 RigVMFunctionLibrary {links [], nodes []}
 RigVMModel {nodes [
   RigUnit_BeginExecution = &RigUnit_BeginExecution()
   SequenceAggregate = &RigVMFunction_Sequence(ExecuteContext=@RigUnit_BeginExecution.ExecutePin)
   n1 = &RigUnit_SphericalPoseReader(
    ActiveRegionScaleFactors=(PositiveWidth=1,NegativeWidth=0.4,PositiveHeight=1.5,NegativeHeight=1.5)
    ActiveRegionSize=0.1
    Debug=(bDrawDebug=true,DebugScale=25,DebugSegments=20,DebugThickness=0.25)
    DriverAxis=(X=0,Y=1,Z=0)
    DriverItem=(Type=Bone,Name="head")
    FalloffRegionScaleFactors=(PositiveWidth=1,NegativeWidth=0.5,PositiveHeight=1.5,NegativeHeight=1.5)
    FalloffSize=0.45
    FlipWidthScaling=false
    OptionalParentItem=(Type=Bone,Name="spine_05")
    RotationOffset=(X=90,Y=0,Z=-45)
    ExecutePin=@SequenceAggregate.A
   )
   > RigVMVariableNode(Value=@n1.OutputParam) {variable=head_SPR_lookup_alpha}
   > &RigUnit_ModifyTransforms(
    ItemToModify="((Item=(Type=Bone,Name=\"dyn_headBack_Base\"),Transform=(Rotation=(X=0,Y=0,Z=0,W=1),Translation=(X=-2,Y=0,Z=0),Scale3D=(X=1,Y=1,Z=1))),(Item=(Type=Bone,Name=\"dyn_headBack_01_r\"),Transform=(Rotation=(X=0,Y=0,Z=0,W=1),Translation=(X=0,Y=0,Z=0),Scale3D=(X=1,Y=1 ...(truncated)"
    ItemToModify.0.Transform.Rotation=&RigVMFunction_MathQuaternionFromEuler(Euler=(X=0,Y=0,Z=0)).Result
    ItemToModify.1.Transform.Rotation=&RigVMFunction_MathQuaternionFromEuler(Euler=(X=0,Y=-5,Z=-5)).Result
    ItemToModify.2.Transform.Rotation=&RigVMFunction_MathQuaternionFromEuler(Euler=(X=0,Y=-5,Z=-5)).Result
    Weight=RigVMVariableNode() {variable=head_SPR_lookup_alpha}.Value
   )
   > n2 = &RigUnit_SphericalPoseReader(
    ActiveRegionScaleFactors=(PositiveWidth=2,NegativeWidth=0.5,PositiveHeight=2,NegativeHeight=2)
    ActiveRegionSize=0.075
    Debug=(bDrawDebug=false,DebugScale=25,DebugSegments=20,DebugThickness=0.25)
    DriverAxis=(X=0,Y=1,Z=0)
    DriverItem=(Type=Bone,Name="head")
    FalloffRegionScaleFactors=(PositiveWidth=1,NegativeWidth=0.4,PositiveHeight=1,NegativeHeight=1)
    FalloffSize=0.3
    FlipHeightScaling=false
    FlipWidthScaling=true
    OptionalParentItem=(Type=Bone,Name="spine_05")
    RotationOffset=(X=90,Y=0,Z=25)
   )
   > &RigUnit_ModifyTransforms(
    ItemToModify="((Item=(Type=Bone,Name=\"dyn_headBack_Base\"),Transform=(Rotation=(X=0,Y=0,Z=0,W=1),Translation=(X=0,Y=0,Z=0),Scale3D=(X=1,Y=1,Z=1))),(Item=(Type=Bone,Name=\"dyn_head_01_Base_r\"),Transform=(Rotation=(X=0,Y=0,Z=0,W=1),Translation=(X=0,Y=0,Z=0),Scale3D=(X=1,Y=1 ...(truncated)"
    ItemToModify.2.Transform.Rotation=&RigVMFunction_MathQuaternionFromEuler(Euler=(X=0,Y=-10,Z=0)).Result
    ItemToModify.1.Transform.Rotation=&RigVMFunction_MathQuaternionFromEuler(Euler=(X=0,Y=10,Z=0)).Result
    ItemToModify.0.Transform.Rotation=&RigVMFunction_MathQuaternionFromEuler(Euler=(X=0,Y=0,Z=-5)).Result
    Weight=@n2.OutputParam
   )
   > &RigUnit_ModifyTransforms(
    ItemToModify="((Item=(Type=Bone,Name=\"dyn_headFront_01_Base_l\"),Transform=(Rotation=(X=0,Y=0,Z=0,W=1),Translation=(X=0,Y=0,Z=0),Scale3D=(X=1,Y=1,Z=1))),(Item=(Type=Bone,Name=\"dyn_headFront_01_Base_r\"),Transform=(Rotation=(X=0,Y=0,Z=0,W=1),Translation=(X=0,Y=0,Z=0),Scale ...(truncated)"
    ItemToModify.0.Transform.Rotation=&RigVMFunction_MathQuaternionFromEuler(Euler=(X=0,Y=-20,Z=0)).Result
    ItemToModify.1.Transform.Rotation=&RigVMFunction_MathQuaternionFromEuler(Euler=(X=0,Y=20,Z=0)).Result
    Weight=@n2.OutputParam
   )
   > n3 = &RigUnit_SphericalPoseReader(
    ActiveRegionScaleFactors=(PositiveWidth=2,NegativeWidth=2,PositiveHeight=3,NegativeHeight=0.5)
    ActiveRegionSize=0.1
    Debug=(bDrawDebug=false,DebugScale=25,DebugSegments=20,DebugThickness=0.25)
    DriverAxis=(X=0,Y=1,Z=0)
    DriverItem=(Type=Bone,Name="head")
    FalloffRegionScaleFactors=(PositiveWidth=1,NegativeWidth=1,PositiveHeight=0,NegativeHeight=0.4)
    FalloffSize=0.27
    FlipHeightScaling=true
    FlipWidthScaling=false
    OptionalParentItem=(Type=Bone)
    RotationOffset=(X=60,Y=0,Z=0)
   )
   > RigVMVariableNode(Value=@n3.OutputParam) {variable=head_SPR_look_right_alpha}
   > &RigUnit_ModifyTransforms(
    ItemToModify=((Item=(Type=Bone,Name="dyn_head_01_Base_r"),Transform=(Rotation=(X=0,Y=0,Z=-0.1305,W=0.9914),Translation=(X=0,Y=0,Z=0),Scale3D=(X=1,Y=1,Z=1))))
    WeightMinimum=0
    Weight=RigVMVariableNode() {variable=head_SPR_look_right_alpha}.Value
   )
   > Spherical Collision = RigVMFunctionReferenceNode(
    Debug=false
    Max Radius=10
    Min Radius=7.5
    Source=(Type=Bone,Name="dyn_head_03_r")
    Source Offset=(X=0,Y=0,Z=0)
    Target=(Type=Bone,Name="spine_05")
    Target Offset=(X=7,Y=0,Z=7.5)
    Weight=1
   )
   n4 = &RigUnit_SphericalPoseReader(
    ActiveRegionScaleFactors=(PositiveWidth=0.6,NegativeWidth=0.6,PositiveHeight=0.8,NegativeHeight=0.8)
    ActiveRegionSize=0.1
    Debug=(bDrawDebug=false,bDraw2D=false,bDrawLocalAxes=true,DebugScale=25,DebugSegments=20,DebugThickness=0.25)
    DriverAxis=(X=-1,Y=0,Z=0)
    DriverItem=(Type=Bone,Name="upperarm_r")
    FalloffRegionScaleFactors=(PositiveWidth=0.6,NegativeWidth=0.6,PositiveHeight=1,NegativeHeight=1)
    FalloffSize=0.2
    FlipHeightScaling=false
    FlipWidthScaling=true
    OptionalParentItem=(Type=Bone,Name="spine_04")
    RotationOffset=(X=-110,Y=-35,Z=50)
    ExecutePin=@SequenceAggregate.B
   )
   > RigVMVariableNode(Value=@n4.OutputParam) {variable=uppearm_r_SPR_aim_up_two_hands_alpha}
   > RigVMVariableNode(Value=&RigVMFunction_MathFloatMul(B=RigVMVariableNode() {variable=uppearm_r_SPR_aim_up_two_hands_alpha}.Value, A=RigVMVariableNode() {variable=head_SPR_lookup_alpha}.Value) {template="Multiply::Execute(in A,in B,out Result)"}.Result) {variable=look_up_and_left_two_hands_multiplied_alpha}
   > &RigUnit_ModifyTransforms(
    ItemToModify="((Item=(Type=Bone,Name=\"dyn_headBack_Base\"),Transform=(Rotation=(X=0,Y=0,Z=0,W=1),Translation=(X=0,Y=0,Z=1),Scale3D=(X=1,Y=1,Z=1))),(Item=(Type=Bone,Name=\"dyn_headBack_01_r\"),Transform=(Rotation=(X=0,Y=0,Z=0,W=1),Translation=(X=-2,Y=-2,Z=0),Scale3D=(X=1,Y= ...(truncated)"
    ItemToModify.0.Transform.Rotation=&RigVMFunction_MathQuaternionFromEuler(Euler=(X=0,Y=10,Z=-10)).Result
    ItemToModify.2.Transform.Rotation=&RigVMFunction_MathQuaternionFromEuler(Euler=(X=-10,Y=-15,Z=-10)).Result
    ItemToModify.1.Transform.Rotation=&RigVMFunction_MathQuaternionFromEuler(Euler=(X=0,Y=-10,Z=0)).Result
    ItemToModify.3.Transform.Rotation=&RigVMFunction_MathQuaternionFromEuler(Euler=(X=0,Y=0,Z=-15)).Result
    Weight=RigVMVariableNode() {variable=look_up_and_left_two_hands_multiplied_alpha}.Value
   )
   > n5 = &RigUnit_SphericalPoseReader(
    ActiveRegionScaleFactors=(PositiveWidth=1,NegativeWidth=1,PositiveHeight=1,NegativeHeight=1)
    ActiveRegionSize=0.075
    Debug=(bDrawDebug=true,DebugScale=25,DebugSegments=20,DebugThickness=0.25)
    DriverAxis=(X=-1,Y=0,Z=0)
    DriverItem=(Type=Bone,Name="upperarm_r")
    FalloffRegionScaleFactors=(PositiveWidth=1,NegativeWidth=1,PositiveHeight=1,NegativeHeight=1)
    FalloffSize=0.2
    OptionalParentItem=(Type=Bone)
    RotationOffset=(X=-55,Y=0,Z=-20)
   )
   > RigVMVariableNode(Value=@n5.OutputParam) {variable=upperarm_r_rifle_up_alpha}
   > RigVMVariableNode(Value=&RigVMFunction_MathFloatMul(B=RigVMVariableNode() {variable=upperarm_r_rifle_up_alpha}.Value, A=RigVMVariableNode() {variable=head_SPR_lookup_alpha}.Value) {template="Multiply::Execute(in A,in B,out Result)"}.Result) {variable=look_up_two_hands_rifle_multiplied_alpha}
   AlphaInterp = &RigVMFunction_AlphaInterp(
    InterpSpeedDecreasing=10
    InterpSpeedIncreasing=10
    Scale=1
    bClampResult=true
    Value=RigVMVariableNode() {variable=look_up_two_hands_rifle_multiplied_alpha}.Value
   )
   > &RigUnit_ModifyTransforms(
    ItemToModify="((Item=(Type=Bone,Name=\"dyn_headBack_Base\"),Transform=(Rotation=(X=-0.0872,Y=0,Z=0,W=0.9962),Translation=(X=0,Y=0,Z=0),Scale3D=(X=1,Y=1,Z=1))),(Item=(Type=Bone,Name=\"dyn_headBack_01_r\"),Transform=(Rotation=(X=0,Y=0,Z=0,W=1),Translation=(X=-3,Y=0,Z=0),Scale ...(truncated)"
    Weight=@AlphaInterp.Result
    ItemToModify.1.Transform.Rotation=&RigVMFunction_MathQuaternionFromEuler(Euler=(X=0,Y=0,Z=0)).Result
    ItemToModify.2.Transform.Rotation=&RigVMFunction_MathQuaternionFromEuler(Euler=(X=0,Y=0,Z=-15)).Result
   )
   /Script/RigVM.RigVMFunction_MathFloatMax(A=RigVMVariableNode() {variable=look_up_and_left_two_hands_multiplied_alpha}.Value) {is=struct RigVMUnitNode, template="Maximum::Execute(in A,in B,out Result)"}
  ]}
]
hierarchy [
 root {loc=<0 0 0>, rot=<0 0 0 1>}
 pelvis {parent=root, loc=<0 -0.5301 98.0571>, rot=<0.0127 0.707 -0.0127 -0.707>}
 spine_01 {parent=pelvis, loc=<6.1208 -0.9328 0>, rot=<0 0 0.034 -0.9994>}
 spine_02 {parent=spine_01, loc=<6.5996 -0.2095 0>, rot=<0 0 0.0238 -0.9997>}
 spine_03 {parent=spine_02, loc=<6.4914 -0.1237 0>, rot=<0 0 0.0149 -0.9999>}
 spine_04 {parent=spine_03, loc=<6.9598 0.0794 0>, rot=<0 0 -0.0132 -0.9999>}
 spine_05 {parent=spine_04, loc=<6.9987 0 0>, rot=<0 0 -0.0015 -1>}
 neck_01 {parent=spine_05, loc=<11.3138 0.1529 0>, rot=<0 0 -0.2003 -0.9797>}
 neck_02 {parent=neck_01, loc=<4.5627 0 0>, rot=<0 0 0.0504 -0.9987>}
 head {parent=neck_02, loc=<5.3914 0 0>, rot=<0 0 0.115 -0.9934>}
 dyn_headFront_01_Base_r {parent=head, loc=<3.953 1.6741 6.5253>, rot=<0.0631 -0.884 0.4623 0.0292>}
 dyn_headFront_01_r {parent=dyn_headFront_01_Base_r, loc=<0 0 0>, rot=<0.9177 0.0617 0.0263 0.3915>}
 dyn_headFront_02_r {parent=dyn_headFront_01_r, loc=<0 0 0>, rot=<0 0 0 1>}
 dyn_headFront_03_r {parent=dyn_headFront_02_r, loc=<0 0 0>, rot=<0 0 0 1>}
 dyn_headFront_01_Base_l {parent=head, loc=<3.2641 2.5389 -6.0881>, rot=<-0.1444 0.6533 0.7368 -0.0971>}
 dyn_headFront_01_l {parent=dyn_headFront_01_Base_l, loc=<0 0 0>, rot=<-0.6634 0.1086 -0.1309 -0.7287>}
 dyn_headFront_02_l {parent=dyn_headFront_01_l, loc=<0 0 0>, rot=<0 0 0 1>}
 dyn_headFront_03_l {parent=dyn_headFront_02_l, loc=<0 0 0>, rot=<0 0 0 1>}
 dyn_headFront_04_l {parent=dyn_headFront_03_l, loc=<0 0 0>, rot=<0 0 0 1>}
 dyn_head_01_Base_l {parent=head, loc=<3.1271 -1.7629 -6.5169>, rot=<0.0456 -0.7633 -0.6422 0.0531>}
 dyn_head_01_l {parent=dyn_head_01_Base_l, loc=<0 0 0>, rot=<0 0 0 1>}
 dyn_head_02_l {parent=dyn_head_01_l, loc=<5.2189 0.1587 0.9066>, rot=<0.0863 0.0823 0.1845 -0.9756>}
 dyn_head_03_l {parent=dyn_head_02_l, loc=<4.1111 0 0>, rot=<0 0 0 1>}
 dyn_head_01_Base_r {parent=head, loc=<2.3393 -2.5791 6.7725>, rot=<0.066 -0.704 0.7045 0.0602>}
 dyn_head_01_r {parent=dyn_head_01_Base_r, loc=<0 0 0>, rot=<0 0 0 1>}
 dyn_head_02_r {parent=dyn_head_01_r, loc=<5.3391 0 0>, rot=<0.0786 0.0273 0.1549 -0.9844>}
 dyn_head_03_r {parent=dyn_head_02_r, loc=<4.1111 0 0>, rot=<0 0 0 1>}
 dyn_frStrand_l_01 {parent=head, loc=<9.5289 7.6914 -3.0816>, rot=<0.1628 -0.9051 -0.3139 -0.236>}
 dyn_frStrand_l_02 {parent=dyn_frStrand_l_01, loc=<3.8467 0 0>, rot=<0 -0.1249 -0.1072 -0.9864>}
 dyn_frStrand_l_03 {parent=dyn_frStrand_l_02, loc=<3.3972 0 0>, rot=<0 0 0 1>}
 dyn_headBack_Lookat {parent=head, loc=<6.9539 -5.7482 -0.0677>, rot=<-0.0528 -0.9986 0 0.0006>}
 dyn_headBack_Base {parent=dyn_headBack_Lookat, loc=<0 0 0>, rot=<-0.0001 0.0006 0.2324 -0.9726>}
 dyn_headBack_01_c {parent=dyn_headBack_Base, loc=<5.269 -0.4697 -0.1062>, rot=<0 0 -0.0203 -0.9998>}
 dyn_headBack_02_c {parent=dyn_headBack_01_c, loc=<3.9528 0 0>, rot=<0 0 -0.0156 -0.9999>}
 dyn_headBack_03_c {parent=dyn_headBack_02_c, loc=<4.6852 0 0>, rot=<0 0 0 1>}
 dyn_headBack_01_l {parent=dyn_headBack_Base, loc=<4.8669 1.4907 5.4464>, rot=<0.4407 0.2004 0.0062 -0.875>}
 dyn_headBack_02_l {parent=dyn_headBack_01_l, loc=<4.1345 0 0>, rot=<-0.0393 -0.0144 0.0264 -0.9988>}
 dyn_headBack_03_l {parent=dyn_headBack_02_l, loc=<4.5827 0 0>, rot=<0 0 0 1>}
 dyn_headBack_01_r {parent=dyn_headBack_Base, loc=<4.6665 2.2033 -5.0424>, rot=<-0.1882 -0.0662 -0.0164 -0.9798>}
 dyn_headBack_02_r {parent=dyn_headBack_01_r, loc=<4.1345 0 0>, rot=<-0.1285 -0.0538 0.0742 -0.9875>}
 dyn_headBack_03_r {parent=dyn_headBack_02_r, loc=<4.5827 0 0>, rot=<0 0 0 1>}
 clavicle_l {parent=spine_05, loc=<5.4831 -0.1053 -3.2946>, rot=<0.0465 -0.7289 -0.009 -0.6829>}
 upperarm_l {parent=clavicle_l, loc=<9.8658 -1.2136 -0.4105>, rot=<0.0164 -0.3245 -0.0269 -0.9454>}
 upperarm_twist_01_l {parent=upperarm_l, loc=<7.2135 -0.1134 0>, rot=<0 0 0 1>}
 deform_upperarm_up_base_l {parent=upperarm_twist_01_l, loc=<-8.891 0.4825 3.9517>, rot=<0 0 0 -1>}
 deform_upperarm_up_l {parent=deform_upperarm_up_base_l, loc=<0 0 0>, rot=<0 0 0 1>}
 clavicle_r {parent=spine_05, loc=<5.4831 -0.1053 3.2946>, rot=<0.7289 0.0465 -0.6829 0.009>}
 upperarm_r {parent=clavicle_r, loc=<-9.8658 1.2136 0.4105>, rot=<0.0164 -0.3245 -0.0269 -0.9454>}
 upperarm_twist_01_r {parent=upperarm_r, loc=<-7.2135 0.1134 0>, rot=<0 0 0 -1>}
 deform_upperarm_up_base_r {parent=upperarm_twist_01_r, loc=<8.891 -0.4825 -3.9517>, rot=<0 0 0 -1>}
 deform_upperarm_up_r {parent=deform_upperarm_up_base_r, loc=<0 0 0>, rot=<0 0 0 1>}
 dyn_hair_TGT {parent=pelvis, loc=<0 0 0>, rot=<0 0 0 1>}
]
variables [
 head_SPR_look_right_alpha {type=double, default=0}
 head_SPR_lookup_alpha {type=double, default=0}
 look_up_and_left_two_hands_multiplied_alpha {type=double, default=0}
 look_up_two_hands_rifle_multiplied_alpha {type=double, default=0}
 uppearm_r_SPR_aim_up_two_hands_alpha {type=double, default=0}
 upperarm_r_rifle_up_alpha {type=double, default=0}
]
