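# Physics asset for the SeleneCobra Stride female character's face accessory, covering back-of-head dynamic bones, named as hair in their shape names.
# Three chains (center, left, right) of three capsule bodies hang from a small Base body, linked through Lookat to spine_05, with tight swing and twist limits.
# Capsules on the upper spine, clavicles and upper arms form the body volume, joined by free-rotation constraints.
# An extra free constraint cross-links the middle right and center segments.
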PhysicsAsset /BRCosmetics/Characters/Player/Female/Medium/Bodies/F_MED_SeleneCobra_Stride/Meshes/Parts/F_MED_SeleneCobra_Stride_FaceAcc_Physics
bodies {
 clavicle_l [{shapeType=Capsule, length=4, radius=6, shapeName=clavicle_l_capsule, center=<6.5 -2 0>, rotation=<rot 16.9211 87.287 -90>}]
 clavicle_r [{shapeType=Capsule, length=4, radius=6, shapeName=clavicle_r_capsule, center=<-6.5 2 0>, rotation=<rot 3.2384 98.0696 -90>}]
 dyn_headBack_01_c [{shapeType=Capsule, length=1.75, radius=3.875, shapeName=dyn_Hair_Back_03_capsule, center=<5 -3 0>, rotation=<rot 90 0 0>}]
 dyn_headBack_01_l [{shapeType=Capsule, length=1.125, radius=4.5, shapeName=dyn_Hair_BackL_03_capsule, center=<5 -2 -3>, rotation=<rot 90 0 0>}]
 dyn_headBack_01_r [
  {shapeType=Capsule, length=2.125, radius=3.5, shapeName=dyn_Hair_BackR_03_capsule, center=<5 -3 2>, rotation=<rot 90 0 0>}
  {
   shapeType=Capsule
   length=1.625
   radius=2.375
   shapeName=dyn_Hair_BackR_03_capsule
   center=<5.8322 -1.8571 -2.9924>
   rotation=<rot 70 -19.9998 0>
  }
 ]
 dyn_headBack_02_c [{shapeType=Capsule, length=1.75, radius=3.875, shapeName=dyn_Hair_Back_04_capsule, center=<5 -3 0>, rotation=<rot 90 0 0>}]
 dyn_headBack_02_l [{shapeType=Capsule, length=1.125, radius=4.5, shapeName=dyn_Hair_BackL_04_capsule, center=<5 -2 -3>, rotation=<rot 90 0 0>}]
 dyn_headBack_02_r [
  {shapeType=Capsule, length=1.875, radius=3.75, shapeName=dyn_Hair_BackR_04_capsule, center=<5 -3 2>, rotation=<rot 90 27.03 27.0299>}
  {
   shapeType=Capsule
   length=2.75
   radius=2.25
   shapeName=dyn_Hair_BackR_04_capsule
   center=<6.6478 -2.6166 -2.6611>
   rotation=<rot 90 180 -175.0001>
  }
 ]
 dyn_headBack_03_c [
  {shapeType=Capsule, length=1.75, radius=3.875, shapeName=dyn_Hair_Back_05_capsule, center=<5 -3 0>, rotation=<rot 90 0 0>}
  {
   shapeType=Capsule
   length=5.625
   radius=2.125
   shapeName=dyn_headBack_03_c_capsule
   center=<8.596 -0.2993 -2.3675>
   rotation=<rot 20.2544 -14.9738 -32.6894>
  }
 ]
 dyn_headBack_03_l [
  {shapeType=Capsule, length=1.125, radius=4.5, shapeName=dyn_Hair_BackL_05_capsule, center=<5 -2 -3>, rotation=<rot 90 0 0>}
  {
   shapeType=Capsule
   length=3.625
   radius=2.25
   shapeName=dyn_headBack_03_l_capsule
   center=<4.8667 0.9841 -7.9272>
   rotation=<rot 75.0001 0.0001 -5.0002>
  }
 ]
 dyn_headBack_03_r [{shapeType=Capsule, length=1.875, radius=3.75, shapeName=dyn_Hair_BackR_05_capsule, center=<5 -3 2>, rotation=<rot 90 27.03 22.0299>}]
 dyn_headBack_Base [{shapeType=Capsule, length=0.505, radius=0.505, shapeName=dyn_Hair_Back_02_capsule}]
 dyn_headBack_Lookat [{shapeType=Capsule, length=0.505, radius=0.505, shapeName=dyn_Hair_Back_01_capsule}]
 spine_04 [{shapeType=Capsule, length=4.7816, radius=9.8876, shapeName=spine_04_capsule, center=<0.3234 0.9034 0>, rotation=<rot 0 -92.6847 0>}]
 spine_05 [
  {shapeType=Capsule, length=13.0661, radius=8.0201, shapeName=spine_05_capsule, center=<0.357 -0.7071 -0.1551>}
  {shapeType=Capsule, length=8.125, radius=6.875, shapeName=spine_05_capsule, center=<8.3314 -1.8232 0>}
 ]
 upperarm_l [{shapeType=Capsule, length=13.4229, radius=4.1429, shapeName=upperarm_l_capsule, center=<5 -1.25 0.0001>, rotation=<rot 0 -90 90>}]
 upperarm_r [{shapeType=Capsule, length=10.4229, radius=4.6429, shapeName=upperarm_r_capsule, center=<-5.5837 1.25 0.0001>, rotation=<rot 0 -90 90>}]
}
constraints [
 {
  bone1Name=spine_05
  bone2Name=spine_04
  swing1LimitDegrees=45
  swing1Motion=Free
  swing2LimitDegrees=45
  swing2Motion=Free
  twistLimitDegrees=45
  twistMotion=Free
 }
 {
  bone1Name=clavicle_l
  bone2Name=spine_05
  swing1LimitDegrees=45
  swing1Motion=Free
  swing2LimitDegrees=45
  swing2Motion=Free
  twistLimitDegrees=45
  twistMotion=Free
 }
 {
  bone1Name=upperarm_l
  bone2Name=clavicle_l
  swing1LimitDegrees=45
  swing1Motion=Free
  swing2LimitDegrees=45
  swing2Motion=Free
  twistLimitDegrees=45
  twistMotion=Free
 }
 {
  bone1Name=clavicle_r
  bone2Name=spine_05
  swing1LimitDegrees=45
  swing1Motion=Free
  swing2LimitDegrees=45
  swing2Motion=Free
  twistLimitDegrees=45
  twistMotion=Free
 }
 {
  bone1Name=upperarm_r
  bone2Name=clavicle_r
  swing1LimitDegrees=45
  swing1Motion=Free
  swing2LimitDegrees=45
  swing2Motion=Free
  twistLimitDegrees=45
  twistMotion=Free
 }
 {
  bone1Name=dyn_headBack_Lookat
  bone2Name=spine_05
  swing1LimitDegrees=45
  swing1Motion=Free
  swing2LimitDegrees=45
  swing2Motion=Free
  twistLimitDegrees=45
  twistMotion=Free
 }
 {
  bone1Name=dyn_headBack_Base
  bone2Name=dyn_headBack_Lookat
  swing1LimitDegrees=45
  swing1Motion=Free
  swing2LimitDegrees=45
  swing2Motion=Free
  twistLimitDegrees=45
  twistMotion=Free
 }
 {bone1Name=dyn_headBack_01_c, bone2Name=dyn_headBack_Base, swing1LimitDegrees=35, swing2LimitDegrees=15, twistLimitDegrees=5}
 {bone1Name=dyn_headBack_02_c, bone2Name=dyn_headBack_01_c, swing1LimitDegrees=35, swing2LimitDegrees=15, twistLimitDegrees=5}
 {bone1Name=dyn_headBack_03_c, bone2Name=dyn_headBack_02_c, swing1LimitDegrees=35, swing2LimitDegrees=15, twistLimitDegrees=5}
 {bone1Name=dyn_headBack_01_l, bone2Name=dyn_headBack_Base, swing1LimitDegrees=35, swing2LimitDegrees=15, twistLimitDegrees=5}
 {bone1Name=dyn_headBack_02_l, bone2Name=dyn_headBack_01_l, swing1LimitDegrees=35, swing2LimitDegrees=15, twistLimitDegrees=5}
 {bone1Name=dyn_headBack_03_l, bone2Name=dyn_headBack_02_l, swing1LimitDegrees=35, swing2LimitDegrees=15, twistLimitDegrees=5}
 {bone1Name=dyn_headBack_01_r, bone2Name=dyn_headBack_Base, swing1LimitDegrees=35, swing2LimitDegrees=15, twistLimitDegrees=5}
 {bone1Name=dyn_headBack_02_r, bone2Name=dyn_headBack_01_r, swing1LimitDegrees=35, swing2LimitDegrees=15, twistLimitDegrees=5}
 {bone1Name=dyn_headBack_03_r, bone2Name=dyn_headBack_02_r, swing1LimitDegrees=35, swing2LimitDegrees=15, twistLimitDegrees=5}
 {
  bone1Name=dyn_headBack_02_r
  bone2Name=dyn_headBack_02_c
  swing1LimitDegrees=45
  swing1Motion=Free
  swing2LimitDegrees=45
  swing2Motion=Free
  twistLimitDegrees=45
  twistMotion=Free
 }
]
